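# Post-pose AnimBP for the female medium Cloaked Assassin body, adding corrective bone motion to the pose copied from its attached parent mesh.
# Each update computes left and right thigh alphas from each thigh's transform delta against the pelvis.
# These alphas drive additive bone-space scaling of the dyn_sub_thigh bones.
# Stacked weighted multi-target Constraint nodes then attach the dyn_crystal_arm bones to left upper-arm/shoulder and right forearm/hand bones.
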AnimBlueprint /BRCosmetics/Characters/Player/Female/Medium/Bodies/F_MED_Cloaked_Assassin/Meshes/F_MED_Cloaked_Assassin_AnimBP
target_skeleton=&/F_MED_Cloaked_Assassin_Skeleton
graphs [
 AnimGraph {nodes [
   n1 = "Transform (Modify) Bone - Bone: dyn_sub_thigh_r"(
    ComponentPose=Local To Component(LocalPose=Copy Pose From Mesh() {bUseAttachedParent=true})
    Alpha=Get Thigh_Rot_R() -> (Thigh_Rot_R:double)
    Translation="0,0,0"
    Rotation="0,0,0"
    Scale="10,4,10"
   ) {
    scaleMode=BMM_Additive
    scaleSpace=BCS_BoneSpace
    alphaBoolBlend {blendOption=Linear}
    boneToModify {boneName=dyn_sub_thigh_r}
    scale=<10 4 10>
   }
   n2 = "Transform (Modify) Bone - Bone: dyn_sub_thigh_l"(ComponentPose=@n1, Alpha=Get Thigh_Rot_L() -> (Thigh_Rot_L:double), Translation="0,0,0", Rotation="0,0,0", Scale="10,4,10") {
    scaleMode=BMM_Additive
    scaleSpace=BCS_BoneSpace
    alphaBoolBlend {blendOption=Linear}
    boneToModify {boneName=dyn_sub_thigh_l}
    scale=<10 4 10>
   }
   n3 = "Constraint - dyn_crystal_arm_l (PPPPP)"!(
    ComponentPose=@n2
    Alpha=1
    ConstraintWeights_0=0.1395
    ConstraintWeights_1:float=0.0147
    ConstraintWeights_2:float=0.1271
    ConstraintWeights_3:float=0.7183
    ConstraintWeights_4:float=0.0004
   ) -> (Pose) {
    alphaBoolBlend {blendOption=Linear}
    boneToModify {boneName=dyn_crystal_arm_l}
    constraintSetup [
     {targetBone {boneName=upperarm_twist_02_l}}
     {targetBone {boneName=clavicle_l}}
     {targetBone {boneName=upperarm_l}}
     {targetBone {boneName=upperarm_twist_01_l}}
     {targetBone {boneName=spine_05}}
    ]
    constraintWeights [1,1,1,1,1]
   }
   n4 = "Constraint - dyn_crystal_arm_l (PPPPP)"(
    ComponentPose=@n3
    Alpha=1
    ConstraintWeights_0=0.203
    ConstraintWeights_1=0.0001
    ConstraintWeights_2=0.1946
    ConstraintWeights_3=0.6022
    ConstraintWeights_4=0
   ) {
    alphaBoolBlend {blendOption=Linear}
    boneToModify {boneName=dyn_crystal_arm_l}
    constraintSetup [
     {targetBone {boneName=upperarm_twist_02_l}}
     {targetBone {boneName=clavicle_l}}
     {targetBone {boneName=upperarm_l}}
     {targetBone {boneName=upperarm_twist_01_l}}
     {targetBone {boneName=spine_05}}
    ]
    constraintWeights [1,1,1,1,1]
   }
   n5 = "Constraint - dyn_crystal_arm_l (PPP)"!(ComponentPose=@n4, Alpha=1, ConstraintWeights_0=0.2014, ConstraintWeights_1:float=0.2033, ConstraintWeights_2:float=0.5954) -> (Pose) {
    alphaBoolBlend {blendOption=Linear}
    boneToModify {boneName=dyn_crystal_arm_l}
    constraintSetup [{targetBone {boneName=upperarm_twist_02_l}}, {targetBone {boneName=upperarm_l}}, {targetBone {boneName=upperarm_twist_01_l}}]
    constraintWeights [1,1,1]
   }
   n6 = "Constraint - dyn_crystal_arm_r (PPPPP)"!(
    ComponentPose=@n5
    Alpha=1
    ConstraintWeights_0=0.1029
    ConstraintWeights_1:float=0
    ConstraintWeights_2:float=0.002
    ConstraintWeights_3:float=0.136
    ConstraintWeights_4:float=0.7591
   ) -> (Pose) {
    alphaBoolBlend {blendOption=Linear}
    boneToModify {boneName=dyn_crystal_arm_r}
    constraintSetup [
     {targetBone {boneName=lowerarm_twist_02_r}}
     {targetBone {boneName=thumb_01_r}}
     {targetBone {boneName=hand_r}}
     {targetBone {boneName=lowerarm_r}}
     {targetBone {boneName=lowerarm_twist_01_r}}
    ]
    constraintWeights [1,1,1,1,1]
   }
   Output Pose(Result=Component To Local(ComponentPose=@n6)) {layerGroup=None}
  ]}
 EventGraph {nodes [
   Event BlueprintUpdateAnimation()
   > Get!(execute:Exec) -> (then:Exec, else:Exec, Skel_Mesh:Skeletal Mesh Component Object Reference)
   > n1 = Sequence()
   > Set Thigh_Rot_L(Thigh_Rot_L:double=GetAlphaFromReferenceTransformDelta(SkeletalMeshComponent=Get Skel_Mesh(), BoneName=thigh_l, BaseName=pelvis, InputRange=(X=-19,Y=-101), TwistAxis="0,0,1")) -> (Output_Get:double)
   Set Thigh_Rot_R(
    execute=@n1.then_1
    Thigh_Rot_R:double=GetAlphaFromReferenceTransformDelta(SkeletalMeshComponent=Get Skel_Mesh(), BoneName=thigh_r, BaseName=pelvis, InputRange=(X=4,Y=-101), TwistAxis="0,0,1")
   ) -> (Output_Get:double)
   Event BlueprintInitializeAnimation()
   > Set Skel_Mesh(Skel_Mesh=GetOwnerSkeletalMeshComponent())
  ]}
]
variables [
 Skel_Mesh {type=SkeletalMeshComponent}
 Thigh_Rot_L {type=number}
 Thigh_Rot_R {type=number}
]
