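# Physics asset for the F_MED_Cloaked_Assassin female medium body part, giving the cloak's hood simulated secondary motion.
# Box bodies sit on the left and right dynamic hood base and parent bones, alongside a head capsule that likely serves as a collider for the hood.
# Each parent bone is constrained to its base with narrow swing and twist limits, so each hood side moves only slightly and independently.
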PhysicsAsset /BRCosmetics/Characters/Player/Female/Medium/Bodies/F_MED_Cloaked_Assassin/Meshes/Parts/F_MED_Cloaked_Assassin_Physics
bodies {
 dyn_hood_l_base [{shapeType=Box, extentX=0.5, extentY=0.5, extentZ=0.5}]
 dyn_hood_l_parent [{shapeType=Box, extentX=1.764, extentY=5.463, extentZ=7.2064, center=<2.7067 0.0001 0>, rotation=<rot 0 -90 0>}]
 dyn_hood_r_base [{shapeType=Box, extentX=0.5, extentY=0.5, extentZ=0.5}]
 dyn_hood_r_parent [{shapeType=Box, extentX=1.9773, extentY=6.4625, extentZ=7.3459, center=<2.7778 -0.8412 0.7785>, rotation=<rot 2.8408 -90 -1.1883>}]
 head [{shapeType=Capsule, length=4.6063, radius=6.7144, center=<-0.16 3.5832 0>, rotation=<rot 90 0 -10.82>}]
}
constraints [
 {bone1Name=dyn_hood_r_parent, bone2Name=dyn_hood_r_base, swing1LimitDegrees=20, swing2LimitDegrees=5, twistLimitDegrees=2}
 {bone1Name=dyn_hood_l_parent, bone2Name=dyn_hood_l_base, swing1LimitDegrees=20, swing2LimitDegrees=5, twistLimitDegrees=2}
]
